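# Looping ambient sound cue for the edge wall of the Sideways biome bubble.
# A quality level node routes its first two slots to one looping bubble wall wave player and leaves the other two empty.
# Bubble wall attenuation and the outdoor ambient sound class shape playback, while a concurrency override caps simultaneous instances and stops the quietest.
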SoundCue /SidewaysGameplay/Sounds/Bubble/Biome_Sideways_Bubble_Edge_Loop_Cue
root=QualityLevel_0
nodes [
 {id=QualityLevel_0, children [WavePlayer_1,WavePlayer_1,null,null]}
 {id=WavePlayer_1, properties {bLooping=true, soundWaveAssetPtr=&/Sideways_Bubble_Wall_Loop_01}}
]
properties {
 attenuationSettings=&/Sideways_Bubble_Wall_Att
 maxDistance=1628.6998
 priority=10
 soundClassObject=&Fort_Ambient_Outdoors_Athena
 volumeMultiplier=0.626
 concurrencyOverrides {maxCount=2, resolutionRule=StopQuietest}
}
