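# Sound cue that loops the 120 BPM Sideways combat music, apparently for the edge of the Sideways biome bubble.
# A single looping wave player is the root, with no attenuation or mixing nodes.
# The cue skips base submix and submix sends, feeding a pre-effect send into the Sideways audio bus.
# High priority, a very large max distance, boosted volume and a small concurrency cap that stops the quietest instance govern playback.
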SoundCue /SidewaysGameplay/Sounds/Bubble/Biome_Sideways_Bubble_Edge_Loop_Music_Cue
root=WavePlayer_2
nodes [{id=WavePlayer_2, properties {bLooping=true, soundWaveAssetPtr=/SidewaysGameplay/Music/S18_Sideways_Combat_Loop_120BPM}}]
properties {
 bEnableBaseSubmix=false
 bEnableSubmixSends=false
 maxDistance=2097152
 priority=100
 volumeMultiplier=1.5
 concurrencyOverrides {maxCount=2, resolutionRule=StopQuietest}
 preEffectBusSends [{audioBus=&/Sideways_AudioBus, soundSourceBus=null}]
}
