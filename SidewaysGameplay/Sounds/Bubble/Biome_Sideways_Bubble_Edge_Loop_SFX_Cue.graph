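# Sound cue for the looping edge sound of the Sideways biome bubble, judging by its name and wave asset.
# A single looping wave player feeds the bubble wall loop directly to the output, with no randomization or mixing.
# The cue boosts volume, uses the bubble wall attenuation asset and routes to the Fort_AGV sound class.
# A concurrency override caps it at two instances, stopping the quietest when the limit is exceeded.
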SoundCue /SidewaysGameplay/Sounds/Bubble/Biome_Sideways_Bubble_Edge_Loop_SFX_Cue
root=WavePlayer_2
nodes [{id=WavePlayer_2, properties {bLooping=true, soundWaveAssetPtr=&/Sideways_Bubble_Wall_Loop_01}}]
properties {
 attenuationSettings=&/Sideways_Bubble_Wall_Att
 maxDistance=1628.6998
 priority=10
 soundClassObject=/Game/Sounds/FortSoundClassesAndModes/Fort_AGV
 volumeMultiplier=1.68
 concurrencyOverrides {maxCount=2, resolutionRule=StopQuietest}
}
