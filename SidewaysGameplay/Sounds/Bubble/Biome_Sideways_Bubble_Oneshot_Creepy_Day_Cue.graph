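# Sound cue for creepy daytime one-shot ambience in the Sideways biome bubble, built from bubble interior sweetener sounds.
# A single Random node picks among twenty equally weighted sweetener wave players.
# Output plays quietly at a reduced volume in the outdoor ambient sound class, spatialized with the tile ambient one-shot attenuation.
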SoundCue /SidewaysGameplay/Sounds/Bubble/Biome_Sideways_Bubble_Oneshot_Creepy_Day_Cue
root=Random_1
nodes [
 {
  id=Random_1
  children [
   WavePlayer_0
   WavePlayer_7
   WavePlayer_8
   WavePlayer_9
   WavePlayer_10
   WavePlayer_11
   WavePlayer_12
   WavePlayer_13
   WavePlayer_14
   WavePlayer_15
   WavePlayer_16
   WavePlayer_17
   WavePlayer_18
   WavePlayer_19
   WavePlayer_20
   WavePlayer_21
   WavePlayer_22
   WavePlayer_23
   WavePlayer_24
   WavePlayer_25
  ]
  properties {weights [1,1,1,1,1,1,1,1,1,1,1,1,1,1,1,1,1,1,1,1]}
 }
 {id=WavePlayer_0, properties {soundWaveAssetPtr=&/Sideways_Bubble_Interior_Sweetener_10}}
 {id=WavePlayer_10, properties {soundWaveAssetPtr=&/Sideways_Bubble_Interior_Sweetener_14}}
 {id=WavePlayer_11, properties {soundWaveAssetPtr=&/Sideways_Bubble_Interior_Sweetener_15}}
 {id=WavePlayer_12, properties {soundWaveAssetPtr=&/Sideways_Bubble_Interior_Sweetener_16}}
 {id=WavePlayer_13, properties {soundWaveAssetPtr=&/Sideways_Bubble_Interior_Sweetener_17}}
 {id=WavePlayer_14, properties {soundWaveAssetPtr=&/Sideways_Bubble_Interior_Sweetener_18}}
 {id=WavePlayer_15, properties {soundWaveAssetPtr=&/Sideways_Bubble_Interior_Sweetener_19}}
 {id=WavePlayer_16, properties {soundWaveAssetPtr=&/Sideways_Bubble_Interior_Sweetener_20}}
 {id=WavePlayer_17, properties {soundWaveAssetPtr=&/Sideways_Bubble_Interior_Sweetener_01}}
 {id=WavePlayer_18, properties {soundWaveAssetPtr=&/Sideways_Bubble_Interior_Sweetener_02}}
 {id=WavePlayer_19, properties {soundWaveAssetPtr=&/Sideways_Bubble_Interior_Sweetener_03}}
 {id=WavePlayer_20, properties {soundWaveAssetPtr=&/Sideways_Bubble_Interior_Sweetener_04}}
 {id=WavePlayer_21, properties {soundWaveAssetPtr=&/Sideways_Bubble_Interior_Sweetener_05}}
 {id=WavePlayer_22, properties {soundWaveAssetPtr=&/Sideways_Bubble_Interior_Sweetener_06}}
 {id=WavePlayer_23, properties {soundWaveAssetPtr=&/Sideways_Bubble_Interior_Sweetener_07}}
 {id=WavePlayer_24, properties {soundWaveAssetPtr=&/Sideways_Bubble_Interior_Sweetener_08}}
 {id=WavePlayer_25, properties {soundWaveAssetPtr=&/Sideways_Bubble_Interior_Sweetener_09}}
 {id=WavePlayer_7, properties {soundWaveAssetPtr=&/Sideways_Bubble_Interior_Sweetener_11}}
 {id=WavePlayer_8, properties {soundWaveAssetPtr=&/Sideways_Bubble_Interior_Sweetener_12}}
 {id=WavePlayer_9, properties {soundWaveAssetPtr=&/Sideways_Bubble_Interior_Sweetener_13}}
]
properties {
 attenuationSettings=/Game/Sounds/Attenuation/Fort_TileAmbient_OneShot_Att
 maxDistance=8000
 soundClassObject=&Fort_Ambient_Outdoors_Athena
 volumeMultiplier=0.2
 preEffectBusSends [{}]
}
