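# Sound cue for the death sound of a reskinned Deimos monster, layering Fiend death effects with Sideways bubble impact sounds.
# A Mixer blends two random branches, with the bubble layer quieter than the death layer.
# The death branch randomly picks one of five Fiend waves and pitches it down through a Modulator, while the other picks one of four bubble impacts.
# The cue overrides attenuation and concurrency, capping simultaneous instances.
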SoundCue /SidewaysGameplay/Sounds/Bubble/Deimos_Reskin_Death_SFX_Cue
root=Mixer_0
nodes [
 {id=Mixer_0, children [Modulator_1,Random_1], properties {inputVolume [0.85,0.5]}}
 {id=Modulator_1, children [Random_0], properties {pitchMax=0.8, pitchMin=0.75}}
 {id=Random_0, children [WavePlayer_0,WavePlayer_1,WavePlayer_2,WavePlayer_3,WavePlayer_4], properties {weights [1,1,1,1,1]}}
 {id=Random_1, children [WavePlayer_5,WavePlayer_6,WavePlayer_7,WavePlayer_8], properties {weights [1,1,1,1]}}
 {id=WavePlayer_0, properties {soundWaveAssetPtr=/Game/Sounds/Fort_Monster_Deimos/SFX/FiendCP_DeathSFX_04}}
 {id=WavePlayer_1, properties {soundWaveAssetPtr=/Game/Sounds/Fort_Monster_Deimos/SFX/FiendCP_DeathSFX_01}}
 {id=WavePlayer_2, properties {soundWaveAssetPtr=/Game/Sounds/Fort_Monster_Deimos/SFX/FiendCP_DeathSFX_02}}
 {id=WavePlayer_3, properties {soundWaveAssetPtr=/Game/Sounds/Fort_Monster_Deimos/SFX/FiendCP_DeathSFX_03}}
 {id=WavePlayer_4, properties {soundWaveAssetPtr=/Game/Sounds/Fort_Monster_Deimos/SFX/FiendCP_DeathSFX_05}}
 {id=WavePlayer_5, properties {soundWaveAssetPtr=&/Sideways_Bubble_Impact_01}}
 {id=WavePlayer_6, properties {soundWaveAssetPtr=&/Sideways_Bubble_Impact_02}}
 {id=WavePlayer_7, properties {soundWaveAssetPtr=&/Sideways_Bubble_Impact_03}}
 {id=WavePlayer_8, properties {soundWaveAssetPtr=&/Sideways_Bubble_Impact_04}}
]
properties {
 bOverrideAttenuation=true
 bOverrideConcurrency=true
 maxDistance=3700
 soundClassObject=/Game/Sounds/FortSoundClassesAndModes/Fort_AGV
 volumeMultiplier=1
 attenuationOverrides {falloffDistance=2500, attenuationShapeExtents {x=1200}}
 concurrencyOverrides {maxCount=3}
}
