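# Sound cue for a Grabber bubble exit sound in the Sideways gameplay set, apparently for a projectile.
# A random node picks one of three Sideways bubble exit wave variations with equal weight.
# It overrides attenuation and concurrency, allowing a single instance and blocking new plays while one is active, and routes to the Fort_AGV sound class.
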SoundCue /SidewaysGameplay/Sounds/Bubble/Grabber_Bubble_Exit_Prj_Cue
root=Random_1
nodes [
 {id=Random_1, children [WavePlayer_7,WavePlayer_8,WavePlayer_9], properties {weights [1,1,1]}}
 {id=WavePlayer_7, properties {soundWaveAssetPtr=&/Sideways_Bubble_Exit_01}}
 {id=WavePlayer_8, properties {soundWaveAssetPtr=&/Sideways_Bubble_Exit_02}}
 {id=WavePlayer_9, properties {soundWaveAssetPtr=&/Sideways_Bubble_Exit_03}}
]
properties {
 bOverrideAttenuation=true
 bOverrideConcurrency=true
 maxDistance=4000
 soundClassObject=/Game/Sounds/FortSoundClassesAndModes/Fort_AGV
 volumeMultiplier=1
 concurrencyOverrides {maxCount=1, resolutionRule=PreventNew}
}
