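# Sound cue for a grabber bubble impact, played as a layered one-shot in the SidewaysGameplay content.
# A mixer plays two random picks together at equal volume: one of four bubble impact waves and one of three bubble exit waves.
# It overrides attenuation for distance falloff, routes to the Fort_AGV sound class, and caps concurrency at a few instances, stopping the oldest.
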SoundCue /SidewaysGameplay/Sounds/Bubble/Grabber_Bubble_Impact_Cue
root=Mixer_0
nodes [
 {id=Mixer_0, children [Random_1,Random_0], properties {inputVolume [1,1]}}
 {id=Random_0, children [WavePlayer_0,WavePlayer_1,WavePlayer_2,WavePlayer_3], properties {weights [1,1,1,1]}}
 {id=Random_1, children [WavePlayer_7,WavePlayer_8,WavePlayer_9], properties {weights [1,1,1]}}
 {id=WavePlayer_0, properties {soundWaveAssetPtr=&/Sideways_Bubble_Impact_01}}
 {id=WavePlayer_1, properties {soundWaveAssetPtr=&/Sideways_Bubble_Impact_02}}
 {id=WavePlayer_2, properties {soundWaveAssetPtr=&/Sideways_Bubble_Impact_03}}
 {id=WavePlayer_3, properties {soundWaveAssetPtr=&/Sideways_Bubble_Impact_04}}
 {id=WavePlayer_7, properties {soundWaveAssetPtr=&/Sideways_Bubble_Exit_01}}
 {id=WavePlayer_8, properties {soundWaveAssetPtr=&/Sideways_Bubble_Exit_02}}
 {id=WavePlayer_9, properties {soundWaveAssetPtr=&/Sideways_Bubble_Exit_03}}
]
properties {
 bOverrideAttenuation=true
 bOverrideConcurrency=true
 maxDistance=4800
 soundClassObject=/Game/Sounds/FortSoundClassesAndModes/Fort_AGV
 attenuationOverrides {attenuationShapeExtents {x=1200}}
 concurrencyOverrides {maxCount=3, resolutionRule=StopOldest}
}
