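# Sound cue for the looping large environmental core sound among the Sideways bubble sounds.
# A single looping wave player feeds an Enveloper whose volume curve fades in quickly, holds near full, then settles to about half over roughly eleven seconds.
# It is routed to the traps sound class, uses in-world forged consumable attenuation, plays slightly pitched down at reduced volume, and keeps playing when silent.
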SoundCue /SidewaysGameplay/Sounds/Bubble/Sideways_EnvCoreLarge_Loop_Cue
root=Enveloper_0
nodes [
 {
  id=Enveloper_0
  children [WavePlayer_1]
  properties {
   pitchCurve {editorCurveData {keys [{time=0}]}}
   volumeCurve {editorCurveData {keys [<at 0 0.01>,<at 0.25 1>,<at 3.964 0.9921>,<at 6.9898 0.7056>,<at 11.397 0.5076>]}}
  }
 }
 {id=WavePlayer_1, properties {bLooping=true, soundWaveAssetPtr=&/Sideways_EnvCoreLarge_Loop_01}}
]
properties {
 attenuationSettings=/Game/Sounds/Attenuation/Fort_ForgedConsumable_InWorld_Att
 maxDistance=1000
 pitchMultiplier=0.95
 soundClassObject=&Fort_Traps
 virtualizationMode=PlayWhenSilent
 volumeMultiplier=0.55
}
