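# Sound cue for the Sideways bubble pillar "res in" effect, layering reused glitch sounds after a short randomized delay.
# A mixer blends a heavy glitch loop enveloped into a brief fading burst, a random NPC glitch hit, and a random glitch-swap sound delayed half a second.
# Modulators vary the first two layers.
# Attenuation and concurrency are overridden on the cue, capping simultaneous instances and preventing new ones beyond that.
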SoundCue /SidewaysGameplay/Sounds/Bubble/Sideways_Pillar_Res_In_Cue
root=Delay_1
nodes [
 {id=Delay_0, children [Random_2], properties {delayMax=0.5, delayMin=0.5}}
 {id=Delay_1, children [Mixer_1], properties {delayMax=1, delayMin=0.75}}
 {
  id=Enveloper_0
  children [WavePlayer_12]
  properties {
   pitchCurve {editorCurveData {keys [{time=0}]}}
   volumeCurve {editorCurveData {keys [<at 0 0>,<at 0.0695 1>,<at 0.5656 1>,<at 0.9167 0.4428>,<at 1.3123 0>]}}
  }
 }
 {id=Mixer_1, children [Modulator_0,Modulator_1,Delay_0], properties {inputVolume [1,1,1]}}
 {id=Modulator_0, children [Enveloper_0]}
 {id=Modulator_1, children [Random_3], properties {pitchMax=0.97}}
 {id=Random_2, children [WavePlayer_4,WavePlayer_5], properties {weights [1,1]}}
 {id=Random_3, children [WavePlayer_6, WavePlayer_10, WavePlayer_11], properties {weights [1,1,1]}}
 {id=WavePlayer_10, properties {soundWaveAssetPtr=/Game/Athena/Sounds/NPCs/NPC_Glitch_VFX_05}}
 {id=WavePlayer_11, properties {soundWaveAssetPtr=/Game/Athena/Sounds/NPCs/NPC_Glitch_VFX_06}}
 {id=WavePlayer_12, properties {soundWaveAssetPtr=/FNGameplayCues/Athena/Sounds/NPCs/NPC_Glitch_VFX_Heavy_Loop}}
 {id=WavePlayer_4, properties {soundWaveAssetPtr=/Game/Athena/Sounds/Items/Consumable_Glitch_Swap_05}}
 {id=WavePlayer_5, properties {soundWaveAssetPtr=/Game/Athena/Sounds/Items/Consumable_Glitch_Swap_04}}
 {id=WavePlayer_6, properties {soundWaveAssetPtr=/Game/Athena/Sounds/NPCs/NPC_Glitch_VFX_01}}
]
properties {
 bOverrideAttenuation=true
 bOverrideConcurrency=true
 maxDistance=4800
 soundClassObject=/Game/Sounds/FortSoundClassesAndModes/Fort_AGV
 volumeMultiplier=0.7
 attenuationOverrides {attenuationShapeExtents {x=1200}}
 concurrencyOverrides {maxCount=2, resolutionRule=PreventNew}
}
